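# Translucent FS Pristinous floor effect material for Niagara meshes, blending scene base color into a faceted floor with glowing cracks via a full-screen transition alpha.
# Scene depth and camera vector reconstruct world-position UVs, projecting the pattern onto existing geometry.
# Heightmap alpha splits panned, distorted emissive cracks from cubemap-lit diffuse, with animated crystal glitch UVs.
# Floor-height fade, custom-stencil mask, dissolve and particle dynamic parameters provide masking and control.
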Material /DealershipSharedLibrary/Assets/Sets/FSPristinous/MM_DLR_FS_Pristinous_Floor
graph {
 properties {
  allowTranslucentCustomDepthWrites=true
  bAutomaticallySetUsageInEditor=false
  bEnableResponsiveAA=true
  bUsedWithNiagaraMeshParticles=true
  blendMode=BLEND_Translucent
  translucencyLightingMode=TLM_VolumetricDirectional
  usedShadingModels=Default Lit
 }
 nodes [
  n1 = TextureCoordinate()
  n2 = ScalarParameter() {defaultValue=1, parameterName=Emission_Glitch_Tile}
  n3 = NamedRerouteUsage()
  n4 = TextureSample(UVs=Add(A=Multiply(A=@n1, B=@n2), B=@n3)) {texture=/DealershipSharedLibrary/Assets/Sets/FSPristinous/Textures/T_DLR_Diamond}
  Desaturation(None=@n4.RGBA)
  DynamicParameter() {defaultValue=<rgba 0 1 0 0>, paramNames [Dissolve,Emissive,Metallic,IsPainted]}
  n5 = ParticleColor()
  n6 = VectorParameter() {parameterName=BaseColor, defaultValue=<rgba 0.0459 0.0645 0.099 1>}
  If(A=NamedRerouteUsage(), "A > B"=Multiply(A=@n5.RGB, B=NamedRerouteUsage()), "A == B"=@n6.RGB, "A < B"=@n6.RGB)
  n7 = NamedRerouteUsage()
  n8 = ScalarParameter() {defaultValue=6800, parameterName=EmissivelTile}
  n9 = Add(A=@n3, B=ComponentMask(None=Divide(A=@n7, B=@n8)) {g=true, r=true})
  n10 = Panner(Coordinate=@n9, Time=Multiply(A=Time(), B=ScalarParameter() {defaultValue=-0.02, parameterName=PanSpeed})) {speedY=0.01}
  n11 = ComponentMask(None=TextureSampleParameter2D(UVs=@n9) {parameterName=EmissiveTextureUV, texture=/DealershipSharedLibrary/FX/Textures/Noise/T_DLR_flowBasic}.RGB) {g=true, r=true}
  n12 = ScalarParameter() {defaultValue=1, parameterName=Emission - Use Diamonds Texture?}
  n13 = ParticleColor()
  n14 = Multiply(A=@n4.RGBA, B=@n13.RGBA)
  n15 = Multiply(
   A=TextureSampleParameter2D(UVs=Add(A=@n10, B=@n11)) {parameterName=Emissive_Ground_Noise_Texture, texture=/DealershipSharedLibrary/Assets/Sets/FSPristinous/Textures/T_DLR_CellDistort}.G
   B=Multiply(
    A=If(A=@n12, "A > B"=@n14, "A == B"=@n13.RGBA, "A < B"=@n13.RGBA)
    B=ScalarParameter() {defaultValue=10, parameterName=EmissiveColor_Intensity}
   )
  )
  n16 = Transform(None=ReflectionVectorWS()) {transformSourceType=TRANSFORMSOURCE_View}
  n17 = Saturate(None=Desaturation(
    None=TextureSampleParameterCube!(UVs=@n16, Apply View MipBias) {parameterName=cubemap, texture=/Game/Packages/Fortress_Sky/TexturesHDR/T_Cubemap_Skylight_Subtle}.RGBA
    Fraction=ScalarParameter() {defaultValue=1, parameterName=desaturation}
   ) {luminanceFactors=<rgba 100 100 100 100>})
  n18 = ScalarParameter() {parameterName="DiffuseTexture - Select Channel? (0 No 1 Yes)"}
  n19 = NamedRerouteUsage()
  n20 = ScalarParameter() {defaultValue=612, parameterName=floor_tile}
  n21 = Add(A=NamedRerouteUsage(), B=ComponentMask(None=Divide(A=@n19, B=@n20)) {g=true, r=true})
  n22 = TextureSampleParameter2D(UVs=Multiply(A=ScalarParameter() {defaultValue=1, parameterName=DiffuseTiling}, B=@n21)) {parameterName=DiffuseTexture, texture=/DealershipSharedLibrary/Assets/Sets/FSPristinous/Textures/T_DLR_CellNoise}
  n23 = ChannelMaskParameter(None=@n22.RGBA) {parameterName=DiffuseTexture - Select Channel}
  n24 = Multiply(
   A=Multiply(A=ComponentMask(None=@n17) {r=true}, B=If(A=@n18, "A > B"=@n23.R, "A == B"=@n22.RGBA, "A < B"=@n22.RGBA)) {constA=1}
   B=Multiply(
    A=VectorParameter() {parameterName=Diffuse_ColorTint, defaultValue=<rgba 0.3185 0.3185 0.3185 1>}.RGBA
    B=ScalarParameter() {defaultValue=1, parameterName=DiffuseColorIntensity}
   )
  )
  n25 = TextureSampleParameter2D(UVs=@n21) {
   parameterName=heightmap
   samplerType=SAMPLERTYPE_Normal
   texture=/DealershipSharedLibrary/Assets/Sets/FSPristinous/Textures/T_DLR_FacetedPattern_N_TEMP
  }
  n26 = Multiply(
   A=Subtract(A=ScalarParameter() {defaultValue=0.4, parameterName=MaskAmount}, B=OneMinus(None=@n25.A))
   B=ScalarParameter() {defaultValue=1, parameterName=MaskSharpness}
  )
  n27 = MaterialFunctionCall!(Use MPC, Is Sprite, Use World Position Behind Translucency) {materialFunction=/DealershipSharedLibrary/FX/Materials/MaterialFunctions/MF_DLR_FullScreenTransition_Alpha}
  n28 = LinearInterpolate(
   A=SceneTexture() {sceneTextureId=PPI_BaseColor}.Color
   B=LinearInterpolate(
    A=NamedRerouteDeclaration(None=@n15) {name=EmissiveGroundCracks, nodeColor=<rgba 1 0.3213 0 1>}
    B=NamedRerouteDeclaration(None=@n24) {name=Diffuse, nodeColor=<rgba 1 0 0.5156 1>}
    Alpha=Saturate(None=@n26)
   )
   Alpha=@n27.Curve - Alpha
  )
  n29 = If(A=NamedRerouteUsage(), "A > B"=Desaturation(None=@n28), "A == B"=@n28, "A < B"=@n28)
  n30 = VectorParameter() {parameterName=OverallTint, defaultValue=<rgba 1 1 1 1>}
  n31 = If(A=NamedRerouteUsage(), "A > B"=@n5.RGBA, "A == B"=@n30.RGBA, "A < B"=@n30.RGBA)
  MaterialFunctionCall(Emissive=Multiply(A=NamedRerouteUsage(), B=Multiply(A=@n29, B=@n31))) {materialFunction=&MF_FX_EmissiveAdjustments}
  n32 = Saturate(None=DepthFade(Opacity=NamedRerouteUsage(), FadeDistance=ScalarParameter() {defaultValue=100, parameterName=depthFade}))
  n33 = Abs(None=Subtract(A=WorldPosition().Z, B=ScalarParameter() {defaultValue=6.4737, parameterName=FloorHeight}))
  n34 = ScalarParameter() {defaultValue=56.2371, parameterName=FadeThreshold}
  n35 = OneMinus(None=Saturate(None=Divide(A=@n33, B=@n34)))
  Multiply(
   A=LinearInterpolate(A=@n32, B=Multiply(A=@n32, B=@n35), Alpha=ScalarParameter() {defaultValue=1.2037, parameterName=DepthFade- Z mask}) {constAlpha=0}
   B=NamedRerouteUsage()
  )
  n36 = SceneTexture() {sceneTextureId=PPI_CustomStencil}
  n37 = Constant() {r=1}
  n38 = DynamicParameter() {parameterIndex=1, defaultValue=<rgba 0 1 1 1>, paramNames [WPO, Base Color Emission Amount, Param3, Param4]}
  n39 = SceneTexture() {sceneTextureId=PPI_CustomDepth}
  n40 = SceneTexture() {sceneTextureId=PPI_SceneDepth}
  n41 = Saturate(None=Subtract(A=ComponentMask(None=@n39.Color) {r=true}, B=ComponentMask(None=@n40.Color) {r=true}))
  NamedRerouteDeclaration(None=If(
    A=ComponentMask(None=@n36.Color) {r=true}
    B=ScalarParameter() {defaultValue=1, group=Stencil Mask, parameterName=Stencil Mask - Stencil Value}
    "A > B"=@n37
    "A == B"=Max(A=@n38.Param3, B=@n41)
    "A < B"=@n37
   )) {name=Stencil Mask, nodeColor=<rgba 1 0 0.0703 1>}
  n42 = ScalarParameter() {defaultValue=12, parameterName=NormalIntensity}
  n43 = Normalize(VectorInput=Transform(None=CameraVectorWS()))
  n44 = ComponentMask(None=SceneDepth().R) {r=true}
  n45 = Constant3Vector() {constant=<rgba 0 0 1 0>}
  n46 = DotProduct(A=@n43, B=Normalize(VectorInput=Transform(None=@n45) {transformSourceType=TRANSFORMSOURCE_View}))
  n47 = Add(A=CameraPositionWS(), B=Multiply(A=@n43, B=Divide(A=@n44, B=@n46)))
  n48 = Divide(A=@n47, B=ScalarParameter() {defaultValue=1, parameterName=DetailNormalTile})
  n49 = TextureSampleParameter2D(UVs=Add(A=NamedRerouteUsage(), B=ComponentMask(None=@n48) {g=true, r=true})) {group=05 Normal, parameterName=Normal - Texture, samplerType=SAMPLERTYPE_Normal, sortPriority=0, texture=/Game/Global/Textures/T_facets_N}
  n50 = VectorParameter() {parameterName=Detail Normal Intensity, defaultValue=<rgba 3 3 1 1>}
  NamedRerouteDeclaration(None=Add(A=Multiply(A=@n25.RGBA, B=@n42), B=Multiply(A=@n49.RGBA, B=@n50.RGBA))) {name=Normal, nodeColor=<rgba 0.5 0.5 1 1>}
  n51 = DynamicParameter() {defaultValue=<rgba 0 1 0 0>, paramNames [Dissolve,Emissive,Metallic,IsPainted]}
  NamedRerouteDeclaration(None=@n51.IsPainted) {name=DP Is Painted, nodeColor=<rgba 0.7344 0 1 1>}
  n52 = NamedRerouteUsage()
  n53 = Multiply(A=@n27.Curve - Alpha, B=@n52)
  n54 = StaticSwitchParameter(
   True=@n53
   False=If(A=ScalarParameter() {parameterName="OpacityMask- Use FullScreenTransition (0-1)"}, "A > B"=@n53, "A == B"=@n52, "A < B"=@n52)
  ) {
   desc="Uses Base Texture's A channel for opacity mask by default (true)"
   group=04 Opacity
   parameterName=Opacity Mask-UseFullScreenTransition
  }
  n55 = StaticSwitchParameter(True=Multiply(A=@n54, B=ParticleColor().A), False=@n54) {
   desc="Uses Base Texture's A channel for opacity mask by default (true)"
   group=04 Opacity
   parameterName=Opacity Mask- Multiply Particle Alpha with Opacity
  }
  NamedRerouteDeclaration(None=MaterialFunctionCall(Opacity In=@n55, DepthFade Distance=ScalarParameter() {group=04 Opacity, parameterName=Opacity - Depth Fade}) {materialFunction=&MF_VCVFX_DepthFade}.Result) {name=Opacity, nodeColor=<rgba 1 1 1 1>}
  NamedRerouteDeclaration(None=@n38.WPO) {name=DP WPO, nodeColor=<rgba 0.7344 0 1 1>}
  n56 = ChannelMaskParameter(None=MaterialFunctionCall!(UV Coord=TextureCoordinate(), LenghtScale, RadialScale) {materialFunction=/Game/Global/MaterialFunctions/TechArt/MF_PolarCoordinates}.Result) {group=06 WPO, parameterName=WPODirection}
  n57 = ScalarParameter() {defaultValue=2, group=06 WPO, parameterName=WPO Spread}
  n58 = Multiply(A=VertexNormalWS(), B=NamedRerouteUsage())
  n59 = Multiply(A=Multiply(A=Power(Base=@n56.R, Exp=@n57), B=ScalarParameter() {defaultValue=1, group=06 WPO, parameterName=WPO Strength}), B=@n58)
  n60 = Multiply(A=@n58, B=ScalarParameter() {defaultValue=1, parameterName=WPOFromNormalInt})
  n61 = ScalarParameter() {parameterName=WPOfromNormalOnly}
  n62 = Constant3Vector()
  NamedRerouteDeclaration(None=If(
    A=ScalarParameter() {defaultValue=1, group=06 WPO, parameterName=WPO ON/OFF, sortPriority=1}
    "A > B"=LinearInterpolate(A=@n59, B=@n60, Alpha=@n61)
    "A == B"=@n62
    "A < B"=@n62
   )) {name=WPO, nodeColor=<rgba 0.5938 0 1 1>}
  NamedRerouteDeclaration(None=MaterialFunctionCall(Alpha Threshold=@n55) {materialFunction=&DitherTemporalAA}.Result) {name=Masked Opacity, nodeColor=<rgba 0.4348 0.4348 0.4348 1>}
  n63 = TextureCoordinate() {coordinateIndex=1}
  n64 = If(A=ScalarParameter() {group=04 Opacity, parameterName=Dissolve - UV Channel}, "A > B"=@n63, "A == B"=@n63, "A < B"=TextureCoordinate())
  n65 = VectorParameter() {group=04 Opacity, parameterName=Dissolve - UVs, channelNames {a=Panner Y, b=Panner X, g=Scale Y, r=Scale X}, defaultValue=<rgba 1 1 0 0>}
  n66 = ComponentMask(None=@n65) {g=true, r=true}
  n67 = AppendVector(A=@n65.Panner X, B=@n65.Panner X)
  n68 = Time()
  n69 = TextureSampleParameter2D(UVs=Add(A=Multiply(A=@n64, B=@n66), B=Multiply(A=@n67, B=@n68))) {
   group=04 Opacity
   parameterName=Dissolve - Texture
   samplerType=SAMPLERTYPE_LinearColor
   sortPriority=30
   texture=/Game/Global/Textures/Default/Noise/T_CloudNoise_01
  }
  n70 = MaterialFunctionCall!(
   Erosion=NamedRerouteUsage()
   Mask=ChannelMaskParameter(None=@n69.RGBA) {group=04 Opacity, parameterName=Dissolve - Channel, sortPriority=31}.R
   Edge Width=ScalarParameter() {group=04 Opacity, parameterName="Dissolve -  Edge Width"}
   Width=Constant() {r=0.25}
  ) {materialFunction=/DealershipSharedLibrary/FX/Materials/MaterialFunctions/MF_DLR_Erode}
  NamedRerouteDeclaration(None=@n70.Edge) {name=Dissolve Edge, nodeColor=<rgba 0.7813 0 1 1>}
  n71 = DynamicParameter() {parameterIndex=3, defaultValue=<rgba 0 1 0 0>, paramNames [Param1, If Painted - Multiply BaseColor, Param3, Param4]}
  NamedRerouteDeclaration(None=@n71.If Painted - Multiply BaseColor) {name=DP Variant Base Color Multiplier, nodeColor=<rgba 0.681 0 1 1>}
  NamedRerouteDeclaration(None=@n38.Base Color Emission Amount) {name=DP Base Color Emission Amount, nodeColor=<rgba 0.681 0 1 1>}
  n72 = Constant4Vector() {constant=<rgba 0.25 1 -0.2 1.25>}
  n73 = TextureCoordinate()
  n74 = ScalarParameter() {defaultValue=8, parameterName=GlitchUVTile}
  n75 = Multiply(A=Multiply(A=@n73, B=AppendVector(A=@n72.R, B=@n72.R)), B=@n74)
  n76 = Time()
  n77 = Multiply(A=ScalarParameter() {defaultValue=0.07, parameterName=MaskGlitchRotSpeed}, B=@n76)
  n78 = Multiply(A=ScalarParameter() {defaultValue=0.02, parameterName=MaskGlitchPanSpeed}, B=@n76)
  n79 = TextureObjectParameter() {parameterName=Master_Glitch_Crystal_UV_Texture, texture=/DealershipSharedLibrary/Assets/Sets/FSPristinous/Textures/T_DLR_Cristal}
  n80 = TextureSample(UVs=Panner(Coordinate=Rotator(Coordinate=@n75, Time=@n77), Time=@n78) {speedX=0.1, speedY=-0.25}, Tex=@n79) {samplerType=SAMPLERTYPE_LinearColor, texture=/Game/Global/MaterialFunctions/Cosmetics/AnimatedGrid/Examples/Textures/T_AnimatedGridMask}
  n81 = Constant4Vector() {constant=<rgba 0.75 0.25 0.5 -0.1>}
  n82 = Multiply(A=Multiply(A=@n73, B=AppendVector(A=@n81.R, B=@n81.R)), B=@n74)
  n83 = TextureSample(UVs=Panner(Coordinate=Rotator(Coordinate=@n82, Time=@n77), Time=@n78) {speedX=-0.25, speedY=0.1}, Tex=@n79)
  n84 = Multiply(A=Multiply(A=@n73, B=AppendVector(A=@n72.B, B=@n72.B)), B=@n74)
  n85 = TextureSample(UVs=Panner(Coordinate=Rotator(Coordinate=@n84, Time=@n77), Time=@n78) {speedX=-0.1, speedY=0.1}, Tex=@n79)
  n86 = Multiply(A=Multiply(A=@n73, B=AppendVector(A=@n81.B, B=@n81.B)), B=@n74)
  n87 = TextureSample(UVs=Panner(Coordinate=Rotator(Coordinate=@n86, Time=@n77), Time=@n78) {speedX=-0.1, speedY=0.1}, Tex=@n79)
  n88 = ScalarParameter() {defaultValue=0.1, parameterName=MaskGlitchIntensity}
  n89 = Multiply(A=@n76, B=ScalarParameter() {defaultValue=0.05, parameterName=FlickerGlitchSpeed})
  n90 = Sine(None=@n89) {period=0.5}
  n91 = Sine(None=@n89) {period=0.25}
  n92 = LinearInterpolate(
   A=LinearInterpolate(A=AppendVector(A=@n80.R, B=@n83.R), B=AppendVector(A=@n85.G, B=@n87.G), Alpha=Multiply(A=@n88, B=@n90))
   B=LinearInterpolate(A=AppendVector(A=@n85.R, B=@n80.R), B=AppendVector(A=@n87.R, B=@n83.R), Alpha=Multiply(A=@n88, B=@n91))
  )
  NamedRerouteDeclaration(None=Saturate(None=Add(A=@n73, B=@n92))) {name=Crystal, nodeColor=<rgba 0 1 0.6953 1>}
  NamedRerouteDeclaration(None=@n47) {name=UVs, nodeColor=<rgba 0.125 0 1 1>}
  n93 = VectorParameter() {group=Time Of Day, parameterName=Color Morning Multiplier, sortPriority=5, defaultValue=<rgba 1 1 1 1>}
  n94 = VectorParameter() {group=Time Of Day, parameterName=Color Day Multiplier, sortPriority=6, defaultValue=<rgba 1 1 1 1>}
  n95 = VectorParameter() {group=Time Of Day, parameterName=Color Evening Multiplier, sortPriority=7, defaultValue=<rgba 1 1 1 0>}
  n96 = VectorParameter() {group=Time Of Day, parameterName=Color Night Multiplier, sortPriority=8, defaultValue=<rgba 1 1 1 0>}
  NamedRerouteDeclaration(None=AppendVector(
    A=MaterialFunctionCall(Color_Morning=@n93.RGB, Color_Day=@n94.RGB, Color_Evening=@n95.RGB, Color_Night=@n96.RGB) {materialFunction=&Color_Time_of_Day}.Result
    B=Constant() {r=1}
   )) {name=TimeOfDay_Color, nodeColor=<rgba 0 0.9811 1 1>}
  NamedRerouteDeclaration(None=@n70.Result) {name=Dissolve, nodeColor=<rgba 0.1875 0.6953 1 1>}
  NamedRerouteDeclaration(None=@n51.Emissive) {name=DP Emissive, nodeColor=<rgba 0.7344 0 1 1>}
  NamedRerouteDeclaration(None=@n51.Metallic) {name=DP Metallic, nodeColor=<rgba 0.7344 0 1 1>}
  NamedRerouteDeclaration(None=@n51.Dissolve) {name=DP Dissolve, nodeColor=<rgba 0.7344 0 1 1>}
  NamedRerouteUsage()
  NamedRerouteUsage()
  NamedRerouteUsage()
  NamedRerouteUsage()
  ScalarParameter() {defaultValue=8, parameterName=GlitchUVTile2}
 ]
}
