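# Temporary translucent, unlit, two-sided material for Niagara mesh particles in the Pristinous set, combining distorted noise with rainbow coloring.
# Two panner-distorted textures multiply into a base mask, shaped by smoothstep dissolve, optional Fresnel and hard cutoff, and depth fade.
# Rainbow color blends counter-panning gradients, a reflection cubemap and a faceted pattern.
# Four dynamic parameters drive tiling, dissolve, distortion and optional noise-based normal displacement, which the quality switch limits to Epic.
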Material /DealershipSharedLibrary/Assets/Sets/FSPristinous/M_DLR_Mesh_NoiseDistortPristinous_TEMP
graph {
 properties {
  bAutomaticallySetUsageInEditor=false
  bUsedWithNiagaraMeshParticles=true
  blendMode=BLEND_Translucent
  refractionMethod=RM_IndexOfRefraction
  shadingModel=MSM_Unlit
  twoSided=true
  usedShadingModels=Unlit
 }
 nodes [
  LinearInterpolate(B=ScalarParameter() {group=Refraction, parameterName=Refraction - Amount}, Alpha=NamedRerouteUsage()) {constA=1, constB=2}
  n1 = NamedRerouteUsage()
  n2 = Constant() {r=1}
  n3 = NamedRerouteUsage()
  MaterialFunctionCall(Emissive=Multiply(A=ParticleColor().RGB, B=If(A=@n1, "A > B"=@n2, "A == B"=@n3, "A < B"=@n3))) {materialFunction=&MF_FX_EmissiveAdjustments}
  n4 = Constant()
  n5 = TextureCoordinate()
  n6 = VectorParameter() {
   group=WPO
   parameterName="WPO -  Noise UV Controls"
   channelNames {a=Panner V, b=Panner U, g=Tiling V, r=Tiling U}
   defaultValue=<rgba 1 1 0 0>
  }
  n7 = AppendVector(A=@n6.Tiling U, B=@n6.Tiling U)
  n8 = TextureSampleParameter2D(UVs=Panner(Coordinate=Multiply(A=@n5, B=@n7) {constB=2}, Speed=AppendVector(A=@n6.Panner U, B=@n6.Panner U)) {speedX=0.2, speedY=0.15}) {group=WPO, parameterName=WPO Noise, samplerType=SAMPLERTYPE_LinearColor, texture=/Game/Effects/SnowSandInteraction/Textures/SSI_Perlin}
  n9 = LinearInterpolate(
   A=ScalarParameter() {defaultValue=-200, group=WPO, parameterName=WPO - Min Amount}
   B=ScalarParameter() {defaultValue=400, group=WPO, parameterName="WPO -Max  Amount"}
   Alpha=@n8.R
  ) {constA=-0.3, constB=0.3}
  n10 = NamedRerouteUsage()
  n11 = Add(
   A=MaterialFunctionCall(Offset Amount=ScalarParameter() {group=WPO, parameterName=WPO - CameraOffset}) {materialFunction=/Engine/Functions/Engine_MaterialFunctions02/WorldPositionOffset/CameraOffset}.Clamped Camera Offset
   B=Multiply(A=VertexNormalWS(), B=Multiply(A=@n9, B=@n10))
  )
  NamedRerouteDeclaration(None=StaticSwitchParameter(True=QualitySwitch(Default=@n4, Low=@n4, Medium=@n4, Epic=@n11), False=Constant()) {group=WPO, parameterName=Use WPO?}) {name=WPO_Out, nodeColor=<rgba 1 0.9098 0.2641 1>}
  n12 = NamedRerouteUsage()
  n13 = ScalarParameter() {parameterName=UseFresnel}
  n14 = Multiply(A=Multiply(A=NamedRerouteUsage(), B=LinearInterpolate(B=@n12, Alpha=@n13) {constA=1}), B=NamedRerouteUsage())
  n15 = ScalarParameter() {defaultValue=0.94, parameterName=width}
  n16 = Constant() {r=1}
  n17 = Constant()
  n18 = Multiply(
   A=ParticleColor().A
   B=LinearInterpolate(A=@n14, B=If(A=@n14, B=@n15, "A > B"=@n16, "A == B"=@n16, "A < B"=@n17), Alpha=ScalarParameter() {parameterName=EnableHardDissolve})
  )
  n19 = ScalarParameter() {defaultValue=100, group=Opacity, parameterName=Depth Fade Distance}
  NamedRerouteDeclaration(None=Multiply(
    A=MaterialFunctionCall(Opacity In=@n18, DepthFade Distance=@n19) {materialFunction=&MF_VCVFX_DepthFade}.Result
    B=ScalarParameter() {defaultValue=1, parameterName=MasterOpacityMultiplier}
   )) {name=Opacity Out, nodeColor=<rgba 0 1 0.9531 1>}
  n20 = DynamicParameter() {parameterIndex=2, defaultValue=<rgba 1 -1 0 0.05>, paramNames [Texture Multiplier, Dissolve Amount, Distortion Amount, WPO Amount]}
  NamedRerouteDeclaration(None=@n20.Distortion Amount) {name=DP_Distortion Amount, nodeColor=<rgba 1 0.0947 0.1282 1>}
  NamedRerouteDeclaration(None=@n20.WPO Amount) {name=DP_WPO Multiplier, nodeColor=<rgba 1 0.0947 0.1282 1>}
  n21 = DynamicParameter() {parameterIndex=3, defaultValue=<rgba 0 0 0 0>, paramNames [DissolveSmoothness, DissolveOffsetX, DissolveOffsetY, IsPainted]}
  NamedRerouteDeclaration(None=@n21.DissolveOffsetY) {name=DP_DissolveOffsetY, nodeColor=<rgba 1 0.0947 0.1282 1>}
  NamedRerouteDeclaration(None=@n21.IsPainted) {name=IsPainted, nodeColor=<rgba 1 0.0947 0.1282 1>}
  n22 = TextureCoordinate()
  n23 = VectorParameter() {parameterName=RainbowUVS, defaultValue=<rgba 0.5 1 0.5 0.5>}
  n24 = Multiply(A=@n22, B=AppendVector(A=@n23.R, B=@n23.R))
  n25 = AppendVector(A=@n23.B, B=@n23.B)
  n26 = TextureObject() {texture=/Game/Global/Textures/Default/Masks_and_Gradients/T_Grad_Rainbow}
  n27 = TextureSample(
   UVs=MaterialFunctionCall(
    UVs=Panner(Coordinate=@n24) {speedY=0.25}
    Rotation Center=@n25
    "Rotation Angle (0-1)"=ScalarParameter() {defaultValue=0.15, parameterName=rot1}
   ) {materialFunction=&CustomRotator}.Rotated Values
   Tex=@n26
  ) {texture=&T_Grad_Rainbow}
  n28 = TextureSample(
   UVs=MaterialFunctionCall(
    UVs=Panner(Coordinate=@n24) {speedY=-0.25}
    Rotation Center=@n25
    "Rotation Angle (0-1)"=ScalarParameter() {defaultValue=-0.15, parameterName=rot2}
   ) {materialFunction=&CustomRotator}.Rotated Values
   Tex=@n26
  ) {texture=&T_Grad_Rainbow}
  n29 = Multiply(A=@n27.RGB, B=@n28.RGB)
  n30 = ReflectionVectorWS()
  n31 = Multiply(
   A=TextureSample(UVs=Transform(None=@n30), Tex=TextureObject() {texture=/DealershipSharedLibrary/FX/Textures/Cubemaps/Cubemap_VCVFX_Rainbow_}) {texture=&Cubemap_VCVFX_Rainbow_}.RGB
   B=@n29
  )
  n32 = ScalarParameter() {defaultValue=1, parameterName=CUBEMAPuse}
  n33 = Panner() {speedX=0.1, speedY=0.3}
  n34 = LinearInterpolate(
   A=@n27.RGB
   B=@n28.RGB
   Alpha=Desaturation(None=TextureSample(UVs=@n33) {samplerType=SAMPLERTYPE_Normal, texture=/DealershipSharedLibrary/Assets/Sets/FSPristinous/Textures/T_DLR_FacetedPattern_N_TEMP}.RGB)
  )
  n35 = ScalarParameter() {defaultValue=0.6025, parameterName=ss}
  NamedRerouteDeclaration(None=Add(A=LinearInterpolate(A=@n29, B=@n31, Alpha=@n32), B=Desaturation(None=@n34, Fraction=@n35) {luminanceFactors=<rgba 0.014 0.062 0.094 0.448>})) {name=RainbowColors, nodeColor=<rgba 0 1 0.5586 1>}
  n36 = NamedRerouteUsage()
  n37 = NamedRerouteUsage()
  n38 = VectorParameter() {parameterName=NoiseUVControls, channelNames {a="", b="", g=ScaleY, r=ScaleX}, defaultValue=<rgba 3 0.5 0 0>}
  n39 = Multiply(A=TextureCoordinate(), B=AppendVector(A=@n38.ScaleX, B=@n38.ScaleX))
  n40 = AppendVector(A=NamedRerouteUsage(), B=NamedRerouteUsage())
  n41 = TextureSampleParameter2D(UVs=Panner(Coordinate=Add(A=@n39, B=@n40))) {group=Dissolve, parameterName=Dissolve - Texture, texture=/Game/Global/Textures/Default/Masks_and_Gradients/T_Water_Top_Ramp}
  n42 = Saturate(None=If(A=ScalarParameter() {defaultValue=1, parameterName=InvertDissolve}, "A > B"=OneMinus(None=@n41.R), "A == B"=@n41.R, "A < B"=@n41.R))
  NamedRerouteDeclaration(None=SmoothStep(Min=Subtract(A=@n36, B=@n37), Max=Add(A=@n36, B=@n37), Value=@n42)) {name=Dissolve Out, nodeColor=<rgba 0.9207 0.4514 1 1>}
  NamedRerouteDeclaration(None=@n21.DissolveSmoothness) {name=DP_DissolveSmoothness, nodeColor=<rgba 1 0.0947 0.1282 1>}
  n43 = NamedRerouteUsage()
  n44 = Multiply(A=NamedRerouteUsage(), B=ScalarParameter() {defaultValue=0.5, group=Texture 2, parameterName=Texture 2 - Distortion Amount})
  n45 = DynamicParameter() {parameterIndex=1, defaultValue=<rgba 15 1 0.75 1>, paramNames [Tex2 Tiling U, Tex2 Tiling V, Tex2 Offset U, Tex2 Offset V]}
  n46 = Multiply(A=@n22, B=AppendVector(A=@n45.Tex2 Tiling U, B=@n45.Tex2 Tiling U))
  n47 = AppendVector(A=@n45.Tex2 Offset U, B=@n45.Tex2 Offset U)
  n48 = MaterialFunctionCall(
   UVs=Add(A=Multiply(A=@n43, B=@n44), B=Add(A=@n46, B=@n47))
   "Rotation Angle (0-1)"=ScalarParameter() {defaultValue=0.25, group=Texture 2, parameterName=Texture 2 - Rotator}
  ) {materialFunction=&CustomRotator}
  n49 = TextureSampleParameter2D(UVs=StaticSwitch(
    True=Clamp(None=@n48.Rotated Values)
    False=@n48.Rotated Values
    Value=StaticBoolParameter() {group=Texture 2, parameterName=Texture 2 - Clamp?}
   )) {group=Texture 2, parameterName=Texture 2, texture=&T_Water_Top_Ramp}
  n50 = LinearInterpolate(
   A=ChannelMaskParameter(None=@n49.RGBA) {group=Texture 2, parameterName=Texture 2 - Channel, sortPriority=11}.R
   B=@n49.RGB
   Alpha=ScalarParameter() {defaultValue=1, group=Texture 2, parameterName=Texture 2 - Use Channel}
  )
  n51 = ScalarParameter() {defaultValue=2, group=Texture 2, parameterName=Texture 2 - Multiplier}
  n52 = NamedRerouteUsage()
  n53 = Multiply(A=NamedRerouteUsage(), B=ScalarParameter() {defaultValue=1, group=Texture 1, parameterName=Texture 1 - Distortion Amount})
  n54 = DynamicParameter() {defaultValue=<rgba 1 17.4562 1 1>, paramNames [Tex1 Tiling U, Tex1 Tiling V, Tex1 Offset U, Tex1 Offset V]}
  n55 = Multiply(A=TextureCoordinate(), B=AppendVector(A=@n54.Tex1 Tiling U, B=@n54.Tex1 Tiling U))
  n56 = AppendVector(A=@n54.Tex1 Offset U, B=@n54.Tex1 Offset U)
  n57 = MaterialFunctionCall(
   UVs=Add(A=Multiply(A=@n52, B=@n53), B=Add(A=@n55, B=@n56))
   "Rotation Angle (0-1)"=ScalarParameter() {group=Texture 1, parameterName=Texture 1 - Rotator}
  ) {materialFunction=&CustomRotator}
  n58 = TextureSampleParameter2D(UVs=StaticSwitch(
    True=Clamp(None=@n57.Rotated Values)
    False=@n57.Rotated Values
    Value=StaticBoolParameter() {group=Texture 1, parameterName=Texture 1 - Clamp?}
   )) {group=Texture 1, parameterName=Texture 1, texture=&T_Water_Top_Ramp}
  n59 = ChannelMaskParameter(None=@n58.RGBA) {group=Texture 1, parameterName=Texture 1 - Channel, sortPriority=11}
  n60 = ScalarParameter() {defaultValue=1, group=Texture 1, parameterName=Texture 1 - Use Channel}
  n61 = Multiply(
   A=Add(A=Multiply(A=@n50, B=@n51), B=ScalarParameter() {defaultValue=0.8, group=Texture 2, parameterName=Texture 2 - Bias})
   B=Multiply(
    A=LinearInterpolate(A=@n59.R, B=@n58.RGB, Alpha=@n60)
    B=ScalarParameter() {defaultValue=0.5644, group=Texture 1, parameterName=Texture 1 - Multiplier}
   )
  )
  NamedRerouteDeclaration(None=Multiply(A=Saturate(None=@n61), B=NamedRerouteUsage())) {name=Base Texture, nodeColor=<rgba 0.0078 1 0 1>}
  NamedRerouteDeclaration(None=@n21.DissolveOffsetX) {name=DP_DissolveOffsetX, nodeColor=<rgba 1 0.0947 0.1282 1>}
  n62 = TextureCoordinate()
  n63 = VectorParameter() {
   group=Texture Distortion
   parameterName=Texture Distortion UV Controls
   sortPriority=5
   channelNames {a=Panner V, b=Panner U, g=Tiling V, r=Tiling U}
   defaultValue=<rgba 1 1 0 0>
  }
  n64 = AppendVector(A=@n63.Tiling U, B=@n63.Tiling U)
  n65 = TextureSampleParameter2D(UVs=Panner(Coordinate=Multiply(A=@n62, B=@n64), Speed=AppendVector(A=@n63.Panner U, B=@n63.Panner U))) {group=Texture Distortion, parameterName=Texture Distortion - Texture, texture=/Game/Effects/Fort_Effects/Textures/Tile/T_PerlinRGB_01_M}
  NamedRerouteDeclaration(None=Subtract(A=ComponentMask(None=@n65.RGB) {g=true, r=true}) {constB=0.5}) {name=Texture Distortion Out, nodeColor=<rgba 0.5435 1 0.2151 1>}
  n66 = VectorParameter() {
   group=Fresnel
   parameterName=Fresnel - Edge
   channelNames {a=Fresnel Bias, b=Fresnel Amount, g=BaseReflectFractionIn, r=ExponentIn}
   defaultValue=<rgba 5 0.04 0 0>
  }
  n67 = Multiply(A=Saturate(None=Fresnel(ExponentIn=@n66.ExponentIn, BaseReflectFractionIn=@n66.ExponentIn)), B=@n66.Fresnel Amount)
  n68 = VectorParameter() {
   group=Fresnel
   parameterName=Fresnel - Interior
   channelNames {a=Fresnel Bias, b=Fresnel Amount, g=BaseReflectFractionIn, r=ExponentIn}
   defaultValue=<rgba 1 0 0 0>
  }
  n69 = Fresnel(ExponentIn=@n68.ExponentIn, BaseReflectFractionIn=@n68.ExponentIn)
  n70 = Multiply(A=Saturate(None=OneMinus(None=@n69)), B=@n68.Fresnel Amount)
  NamedRerouteDeclaration(None=Add(A=Add(A=@n67, B=@n66.Fresnel Bias), B=Add(A=@n70, B=@n68.Fresnel Bias))) {name=Fresnel Out, nodeColor=<rgba 0.1892 0.1821 1 1>}
  NamedRerouteDeclaration(None=@n20.Texture Multiplier) {name=DP_Texture Multiplier, nodeColor=<rgba 1 0.0947 0.1282 1>}
  NamedRerouteDeclaration(None=@n20.Dissolve Amount) {name=DP_Dissolve Amount, nodeColor=<rgba 1 0.0947 0.1282 1>}
  NamedRerouteUsage()
  NamedRerouteUsage()
  Panner() {speedY=0.25}
  n71 = Saturate(None=@n36)
  Saturate(None=Divide(B=Subtract(A=@n42, B=@n71)) {constB=1})
  VectorParameter() {parameterName=Color, defaultValue=<rgba 1 1 1 1>}
 ]
}
