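# Masked material for the Pristinous set's orb and rings, built for Niagara mesh particles, with faked metal shading and a glowing dissolve.
# A gold texture mask blends white and dark colors, optionally tinted by particle color, then multiplies them by a desaturated HDR sky reflection in emissive.
# A noise-driven dissolve adds a bright edge.
# Dynamic parameters drive glow, dissolve, edge emission and paint state, and "notorb" switches between orb and ring normal maps.
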Material /DealershipSharedLibrary/Assets/Sets/FSPristinous/M_DLR_Pristinous_ORB
graph {
 properties {
  bAllowFrontLayerTranslucency=false
  bAutomaticallySetUsageInEditor=false
  bCastRayTracedShadows=false
  bEnableResponsiveAA=true
  bUseTranslucencyVertexFog=false
  bUsedWithNiagaraMeshParticles=true
  blendMode=BLEND_Masked
  translucencyLightingMode=TLM_VolumetricDirectional
  translucencyPass=MTP_AfterMotionBlur
  usedShadingModels=Default Lit
 }
 nodes [
  Constant() {r=2}
  n1 = TextureCoordinate()
  n2 = LinearInterpolate(
   A=TextureSampleParameter2D(UVs=@n1) {parameterName=wsn, texture=/DealershipSharedLibrary/Assets/Sets/FSPristinous/Textures/T_DLR_WSN_PristinousOrb}.RGB
   B=TextureSampleParameter2D() {parameterName=wsn2, texture=/DealershipSharedLibrary/Assets/Sets/FSPristinous/Textures/T_DLR_WSN_PristinousRings}.RGB
   Alpha=ScalarParameter() {parameterName=notorb}
  )
  DeriveNormalZ(InXY=ComponentMask(None=@n2) {g=true, r=true})
  n3 = DynamicParameter() {defaultValue=<rgba 1 3.5209 1 0>, paramNames [GlowIntensity, Dissolve, DissolveEdgeEmission, IsPainted]}
  n4 = NamedRerouteUsage()
  n5 = VectorParameter() {parameterName=WhiteColor, defaultValue=<rgba 0.0729 0.0729 0.0729 1>}
  n6 = ParticleColor()
  n7 = LinearInterpolate(
   A=@n5.RGB
   B=Multiply(A=@n6.RGB, B=ScalarParameter() {defaultValue=30, parameterName=WHITESHELLtint Brightness})
   Alpha=ScalarParameter() {defaultValue=0.8, parameterName=WhiteShellTINT}
  )
  n8 = NamedRerouteUsage()
  n9 = VectorParameter() {parameterName=DarkColor, defaultValue=<rgba 0.0729 0.0268 0.0018 1>}
  n10 = TextureSample(UVs=@n1) {texture=/DealershipSharedLibrary/Assets/Sets/FSPristinous/Textures/T_DLR_Gold}
  n11 = LinearInterpolate(A=@n10.R, B=@n10.R, Alpha=ScalarParameter() {parameterName=notorb})
  n12 = LinearInterpolate(
   A=Multiply(A=@n3.GlowIntensity, B=If(A=@n4, "A > B"=@n7, "A == B"=@n5.RGB, "A < B"=@n5.RGB))
   B=Multiply(A=If(A=@n8, "A > B"=@n6.RGB, "A == B"=@n9.RGB, "A < B"=@n9.RGB), B=ScalarParameter() {defaultValue=1, parameterName=DarkColorMult})
   Alpha=@n11
  )
  n13 = Transform(None=ReflectionVectorWS() {bNormalizeCustomWorldNormal=true})
  n14 = TextureObject() {texture=/Game/Packages/Fortress_Sky/TexturesHDR/T_hdrsky_test03_low}
  n15 = Multiply(
   A=Desaturation(
    None=TextureSample(UVs=@n13, Tex=@n14) {texture=/Game/ArtTools/CubeMapCapture/RT_CubeCapture_01}.RGB
    Fraction=ScalarParameter() {defaultValue=1, parameterName=desaturateNew}
   )
   B=ScalarParameter() {defaultValue=30.48, parameterName=Metalness}
  )
  n16 = ConstantBiasScale(None=Desaturation(None=@n15) {luminanceFactors=<rgba 0.3 0.59 0.11 0>}) {bias=-0.75, scale=2}
  n17 = Transform(None=@n2) {transformSourceType=TRANSFORMSOURCE_World, transformType=TRANSFORM_Local}
  n18 = Saturate(None=Subtract(B=ComponentMask(None=@n17) {b=true}) {constA=0})
  n19 = Multiply(A=ConstantBiasScale(None=Saturate(None=@n16)) {bias=0.25, scale=1.75}, B=Multiply(A=@n15, B=OneMinus(None=@n18)))
  n20 = VectorParameter() {parameterName=DissolveBorder, defaultValue=<rgba 3 0.8438 0.01 1>}
  n21 = Multiply(A=If(A=NamedRerouteUsage(), "A > B"=@n6.RGB, "A == B"=@n20.RGB, "A < B"=@n20.RGB), B=@n3.DissolveEdgeEmission)
  n22 = Power(Base=NamedRerouteUsage(), Exp=ScalarParameter() {defaultValue=5, parameterName=DissolveEdgePower})
  MaterialFunctionCall(Emissive=Add(A=Multiply(A=@n12, B=@n19), B=Multiply(A=@n21, B=@n22))) {materialFunction=&MF_FX_EmissiveAdjustments}
  Multiply(A=@n5.RGB, B=@n6.RGB)
  Multiply(A=@n11, B=ScalarParameter() {defaultValue=1, parameterName=SpecMultiply})
  Multiply(A=@n11, B=ScalarParameter() {defaultValue=13.8917, parameterName="spec power (roughness)"})
  n23 = TextureCoordinate() {uTiling=0.5, vTiling=0.5}
  n24 = VectorParameter() {parameterName=TextureTileOffset, defaultValue=<rgba 1 1 0 0>}
  n25 = AppendVector(A=@n24.R, B=@n24.R)
  n26 = TextureSample(UVs=Panner(Coordinate=Multiply(A=@n23, B=@n25), Speed=AppendVector(A=@n24.B, B=@n24.B))) {texture=&2ndInfectedMapLinear}
  n27 = Constant()
  n28 = Constant() {r=1}
  n29 = Power(Base=ScalarParameter() {defaultValue=2, parameterName=Strange}, Exp=ScalarParameter() {defaultValue=2, parameterName=StrangePower}) {constExponent=3}
  n30 = OneMinus(None=@n29)
  n31 = OneMinus(None=@n3.Dissolve)
  n32 = Constant()
  n33 = Constant() {r=1}
  n34 = Multiply(A=ScalarParameter() {defaultValue=1, parameterName=Remap_Dissolve}, B=ScalarParameter() {defaultValue=12, parameterName=Mult-RemapDissolve})
  n35 = OneMinus(None=@n34)
  n36 = Add(
   A=ComponentMask(None=MaterialFunctionCall(Input=@n26.R, Input Low=@n27, Input High=@n28, Target Low=@n30, Target High=@n29) {materialFunction=&RemapValueRange}.Result) {r=true}
   B=ComponentMask(None=MaterialFunctionCall(Input=@n31, Input Low=@n32, Input High=@n33, Target Low=@n35, Target High=@n34) {materialFunction=&RemapValueRange}.Result) {r=true}
  )
  n37 = Constant() {r=1}
  n38 = If(A=Constant() {r=0.8}, B=Saturate(None=@n36), "A > B"=@n37, "A == B"=@n37, "A < B"=Constant())
  NamedRerouteDeclaration(None=@n38) {name=Dissolve, nodeColor=<rgba 0 0.7169 1 1>}
  n39 = ScalarParameter() {defaultValue=1.6, parameterName=Line}
  n40 = Multiply(A=@n39, B=OneMinus(None=@n39))
  n41 = Constant() {r=1}
  n42 = Constant()
  NamedRerouteDeclaration(None=Multiply(A=If(A=@n36, B=@n40, "A > B"=@n41, "A == B"=@n41, "A < B"=@n42), B=@n38)) {name=DissolveEdge, nodeColor=<rgba 0 1 0.6719 1>}
  NamedRerouteDeclaration(None=@n3.IsPainted) {name=IsPainted, nodeColor=<rgba 0 0.1562 1 1>}
  NamedRerouteUsage()
  ScalarParameter() {defaultValue=0.37, parameterName=desaturateNewbackup}
  ScalarParameter() {defaultValue=0.7089, parameterName=Dissolve}
  ScalarParameter() {defaultValue=1.6, parameterName=Line2}
 ]
}
